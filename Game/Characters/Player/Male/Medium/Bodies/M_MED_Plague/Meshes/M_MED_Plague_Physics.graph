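# Physics asset for the M_MED_Plague medium male player body, covering only the legs.
# Each thigh and calf bone carries one capsule body of matching size, mirrored between left and right.
# Two constraints join each calf to its thigh, with swing and twist left free, so the stored angular limits do not restrict the knees.
# The pelvis, torso, arms, head and feet have no bodies.
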PhysicsAsset /Game/Characters/Player/Male/Medium/Bodies/M_MED_Plague/Meshes/M_MED_Plague_Physics
bodies {
 calf_l [{shapeType=Capsule, length=40, radius=7, center=<-20.957 -0.0004 0>, rotation=<rot 90 0 0>}]
 calf_r [{shapeType=Capsule, length=40, radius=7, center=<20 0 0>, rotation=<rot 90 0 0>}]
 thigh_l [{shapeType=Capsule, length=40, radius=7, center=<-19.9998 2.0958 0.0024>, rotation=<rot 90 0 -179.9999>}]
 thigh_r [{shapeType=Capsule, length=40, radius=7, center=<20 -2.0958 -0.0005>, rotation=<rot 90 0 0>}]
}
constraints [
 {
  bone1Name=calf_r
  bone2Name=thigh_r
  swing1LimitDegrees=45
  swing1Motion=Free
  swing2LimitDegrees=45
  swing2Motion=Free
  twistLimitDegrees=45
  twistMotion=Free
 }
 {
  bone1Name=calf_l
  bone2Name=thigh_l
  swing1LimitDegrees=45
  swing1Motion=Free
  swing2LimitDegrees=45
  swing2Motion=Free
  twistLimitDegrees=45
  twistMotion=Free
 }
]
