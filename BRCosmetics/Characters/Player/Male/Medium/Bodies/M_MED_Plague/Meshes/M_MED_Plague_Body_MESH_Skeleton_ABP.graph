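# Post-copy AnimBlueprint for the male Plague body cosmetic, adding secondary motion to its vial, strings, apron and coattails.
# It copies the attached parent mesh's pose, then caches it after bone constraints and AnimDynamics for vial, strings and apron ties.
# With bClothEnabled false, sprung AnimDynamics chains also simulate apron and coattails, using leg-driven limits on lower coattails; otherwise the cached pose passes through.
# Look At and RigidBody nodes are unconnected.
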AnimBlueprint /BRCosmetics/Characters/Player/Male/Medium/Bodies/M_MED_Plague/Meshes/M_MED_Plague_Body_MESH_Skeleton_ABP
target_skeleton=/Game/Characters/Player/Male/Medium/Bodies/M_MED_Plague/Meshes/M_MED_Plague_Body_MESH_Skeleton
graphs [
 AnimGraph {nodes [
   Anim Dynamics - Bone: dyn_coattail_01_r(Alpha=1, ExternalForce="0,0,0") {
    bGravityOverrideInSimSpace=true
    gravityScale=0.1
    alphaBoolBlend {blendOption=Linear}
    boundBone {boneName=dyn_coattail_01_r}
    physicsBodyDefinitions [{
      sphereCollisionRadius=0
      boundBone {boneName=dyn_coattail_01_r}
      boxExtents=<5 10 10>
      constraintSetup {angularLimitsMax=<0 0 60>, angularLimitsMin=<0 0 0>, angularTarget=<0 0 0>}
      localJointOffset=<5 0 0>
     }]
   }
   "Constraint - dyn_string_l (R)"(Alpha=1, ConstraintWeights_0=1) {
    alphaBoolBlend {blendOption=Linear}
    boneToModify {boneName=dyn_string_l}
    constraintSetup [{transformType=Rotation, targetBone {boneName=dyn_string_r}}]
    constraintWeights [1]
   }
   n1 = Look At - Bone: dyn_apron_master_r(Alpha=1) {
    alphaBoolBlend {blendOption=Linear}
    boneToModify {boneName=dyn_apron_master_r}
    lookAtLocation {x=0}
    lookAtTarget {boneReference {boneName=dyn_apron_master_l}}
    lookAt_Axis {axis {y=0, z=1}}
   }
   n2 = Look At - Bone: dyn_apron_master_l(ComponentPose=@n1, Alpha=1) {
    alphaBoolBlend {blendOption=Linear}
    boneToModify {boneName=dyn_apron_master_l}
    lookAtLocation {x=0}
    lookAtTarget {boneReference {boneName=dyn_apron_master_r}}
    lookAt_Axis {axis {y=0, z=-1}}
   }
   n3 = Look At - Bone: dyn_apron_02_r(ComponentPose=@n2, Alpha=0.5) {
    alphaBoolBlend {blendOption=Linear}
    boneToModify {boneName=dyn_apron_02_r}
    lookAtLocation {x=0}
    lookAtTarget {boneReference {boneName=dyn_apron_02_l}}
    lookAt_Axis {axis {y=0, z=1}}
   }
   n4 = Look At - Bone: dyn_apron_02_l(ComponentPose=@n3, Alpha=0.5) {
    alphaBoolBlend {blendOption=Linear}
    boneToModify {boneName=dyn_apron_02_l}
    lookAtLocation {x=0}
    lookAtTarget {boneReference {boneName=dyn_apron_02_r}}
    lookAt_Axis {axis {y=0, z=-1}}
   }
   n5 = Look At - Bone: dyn_apron_03_r(ComponentPose=@n4, Alpha=0.5) {
    alphaBoolBlend {blendOption=Linear}
    boneToModify {boneName=dyn_apron_03_r}
    lookAtLocation {x=0}
    lookAtTarget {boneReference {boneName=dyn_apron_03_l}}
    lookAt_Axis {axis {y=0, z=1}}
   }
   Look At - Bone: dyn_apron_03_l(ComponentPose=@n5, Alpha=0.5) {
    alphaBoolBlend {blendOption=Linear}
    boneToModify {boneName=dyn_apron_03_l}
    lookAtLocation {x=0}
    lookAtTarget {boneReference {boneName=dyn_apron_03_r}}
    lookAt_Axis {axis {y=0, z=-1}}
   }
   RigidBody() {alphaBoolBlend {blendOption=Linear}}
   RigidBody() {
    bOverrideWorldGravity=true
    overridePhysicsAsset=/Game/Characters/Player/Male/Medium/Bodies/M_MED_Plague/Meshes/M_MED_Plague_Physics
    alphaBoolBlend {blendOption=Linear}
    componentLinearAccScale=<1 1 1>
    componentLinearVelScale=<1 1 1>
    overrideWorldGravity {z=-980}
   }
   n6 = Anim Dynamics - Bone: dyn_coattail_01_r(ComponentPose=Local To Component(LocalPose=Use cached pose 'CopyPoseCache'()), Alpha=1, ExternalForce="0,0,0") {
    angularDampingOverride=0.999
    angularSpringConstant=200
    bAngularSpring=true
    bGravityOverrideInSimSpace=true
    bOverrideAngularDamping=true
    gravityScale=0.35
    simulationSpace=RootRelative
    alphaBoolBlend {blendOption=Linear}
    boundBone {boneName=dyn_coattail_01_r}
    componentLinearAccScale=<0.5 0.5 0.5>
    componentLinearVelScale=<0.5 0.5 0.5>
    physicsBodyDefinitions [{
      sphereCollisionRadius=0
      boundBone {boneName=dyn_coattail_01_r}
      boxExtents=<5 20 20>
      constraintSetup {angularLimitsMax=<0 20 0>, angularLimitsMin=<0 -20 -60>, angularTarget=<0 0 0>}
      localJointOffset=<5 0 0>
     }]
   }
   n7 = Anim Dynamics - Bone: dyn_coattail_02_r(ComponentPose=@n6, Alpha=1, ExternalForce="0,0,0") {
    angularDampingOverride=0.98
    angularSpringConstant=200
    bAngularSpring=true
    bGravityOverrideInSimSpace=true
    bOverrideAngularDamping=true
    gravityScale=0.35
    simulationSpace=RootRelative
    alphaBoolBlend {blendOption=Linear}
    boundBone {boneName=dyn_coattail_02_r}
    physicsBodyDefinitions [{
      sphereCollisionRadius=1
      boundBone {boneName=dyn_coattail_02_r}
      boxExtents=<5 20 20>
      constraintSetup {angularLimitsMax=<0 20 0>, angularLimitsMin=<0 -20 -60>, angularTarget=<0 0 0>}
      localJointOffset=<5 0 0>
     }]
    planarLimits [
     {drivingBone {boneName=thigh_r}, planeTransform {location=<0 10 0>, rotation=<rot 0 0 90>}}
     {drivingBone {boneName=calf_r}, planeTransform {location=<0 11 0>, rotation=<rot 0 0 80>}}
    ]
    sphericalLimits [
     {limitRadius=10, drivingBone {boneName=thigh_l}, sphereLocalOffset=<-17 0 0>}
     {limitRadius=10, drivingBone {boneName=thigh_l}, sphereLocalOffset=<-25 0 0>}
     {limitRadius=8, drivingBone {boneName=thigh_l}, sphereLocalOffset=<-33 0 0>}
     {limitRadius=10, drivingBone {boneName=thigh_r}, sphereLocalOffset=<17 0 0>}
     {limitRadius=10, drivingBone {boneName=thigh_r}, sphereLocalOffset=<25 0 0>}
     {limitRadius=8, drivingBone {boneName=thigh_r}, sphereLocalOffset=<33 0 0>}
     {limitRadius=10, drivingBone {boneName=calf_l}, sphereLocalOffset=<-8 0 0>}
     {limitRadius=10, drivingBone {boneName=calf_l}, sphereLocalOffset=<-16 0 0>}
     {limitRadius=10, drivingBone {boneName=calf_l}, sphereLocalOffset=<-24 0 0>}
     {limitRadius=10, drivingBone {boneName=calf_r}, sphereLocalOffset=<8 0 0>}
     {limitRadius=10, drivingBone {boneName=calf_r}, sphereLocalOffset=<16 0 0>}
     {limitRadius=10, drivingBone {boneName=calf_r}, sphereLocalOffset=<24 0 0>}
    ]
   }
   n8 = Anim Dynamics - Bone: dyn_coattail_03_r(ComponentPose=@n7, Alpha=1, ExternalForce="0,0,0") {
    angularDampingOverride=0.98
    angularSpringConstant=200
    bAngularSpring=true
    bGravityOverrideInSimSpace=true
    bOverrideAngularDamping=true
    gravityScale=0.35
    simulationSpace=RootRelative
    alphaBoolBlend {blendOption=Linear}
    boundBone {boneName=dyn_coattail_03_r}
    physicsBodyDefinitions [{
      sphereCollisionRadius=1
      boundBone {boneName=dyn_coattail_03_r}
      boxExtents=<5 20 20>
      constraintSetup {angularLimitsMax=<0 20 0>, angularLimitsMin=<0 -20 -60>, angularTarget=<0 0 0>}
      localJointOffset=<5 0 0>
     }]
    planarLimits [
     {drivingBone {boneName=thigh_r}, planeTransform {location=<0 10 0>, rotation=<rot 0 0 90>}}
     {drivingBone {boneName=calf_r}, planeTransform {location=<0 10 0>, rotation=<rot 0 0 90>}}
    ]
    sphericalLimits [
     {limitRadius=10, drivingBone {boneName=thigh_l}, sphereLocalOffset=<-17 0 0>}
     {limitRadius=10, drivingBone {boneName=thigh_l}, sphereLocalOffset=<-25 0 0>}
     {limitRadius=8, drivingBone {boneName=thigh_l}, sphereLocalOffset=<-33 0 0>}
     {limitRadius=10, drivingBone {boneName=thigh_r}, sphereLocalOffset=<17 0 0>}
     {limitRadius=10, drivingBone {boneName=thigh_r}, sphereLocalOffset=<25 0 0>}
     {limitRadius=8, drivingBone {boneName=thigh_r}, sphereLocalOffset=<33 0 0>}
     {limitRadius=10, drivingBone {boneName=calf_l}, sphereLocalOffset=<-8 0 0>}
     {limitRadius=10, drivingBone {boneName=calf_l}, sphereLocalOffset=<-16 0 0>}
     {limitRadius=10, drivingBone {boneName=calf_l}, sphereLocalOffset=<-24 0 0>}
     {limitRadius=10, drivingBone {boneName=calf_r}, sphereLocalOffset=<8 0 0>}
     {limitRadius=10, drivingBone {boneName=calf_r}, sphereLocalOffset=<16 0 0>}
     {limitRadius=10, drivingBone {boneName=calf_r}, sphereLocalOffset=<24 0 0>}
    ]
   }
   n9 = Anim Dynamics - Bone: dyn_apron_01_r(ComponentPose=@n8, Alpha=1, ExternalForce="0,0,0") {
    angularDampingOverride=0.999
    angularSpringConstant=200
    bAngularSpring=true
    bGravityOverrideInSimSpace=true
    bOverrideAngularDamping=true
    gravityScale=0.35
    simulationSpace=RootRelative
    alphaBoolBlend {blendOption=Linear}
    boundBone {boneName=dyn_apron_01_r}
    componentLinearAccScale=<0.5 0.5 0.5>
    componentLinearVelScale=<0.5 0.5 0.5>
    physicsBodyDefinitions [{
      sphereCollisionRadius=0
      boundBone {boneName=dyn_apron_01_r}
      boxExtents=<5 20 20>
      constraintSetup {angularLimitsMax=<0 0 60>, angularLimitsMin=<0 0 0>, angularTarget=<0 0 0>}
      localJointOffset=<5 0 0>
     }]
   }
   n10 = Anim Dynamics - Bone: dyn_apron_02_r(ComponentPose=@n9, Alpha=1, ExternalForce="0,0,0") {
    angularDampingOverride=0.98
    angularSpringConstant=200
    bAngularSpring=true
    bGravityOverrideInSimSpace=true
    bOverrideAngularDamping=true
    gravityScale=0.35
    simulationSpace=RootRelative
    alphaBoolBlend {blendOption=Linear}
    boundBone {boneName=dyn_apron_02_r}
    physicsBodyDefinitions [{
      sphereCollisionRadius=0
      boundBone {boneName=dyn_apron_02_r}
      boxExtents=<5 20 20>
      constraintSetup {angularLimitsMax=<0 20 60>, angularLimitsMin=<0 -20 0>, angularTarget=<0 0 0>}
      localJointOffset=<5 0 0>
     }]
   }
   n11 = Anim Dynamics - Bone: dyn_apron_03_r(ComponentPose=@n10, Alpha=1, ExternalForce="0,0,0") {
    angularDampingOverride=0.98
    angularSpringConstant=200
    bAngularSpring=true
    bGravityOverrideInSimSpace=true
    bOverrideAngularDamping=true
    gravityScale=0.35
    simulationSpace=RootRelative
    alphaBoolBlend {blendOption=Linear}
    boundBone {boneName=dyn_apron_03_r}
    physicsBodyDefinitions [{
      sphereCollisionRadius=0
      boundBone {boneName=dyn_apron_03_r}
      boxExtents=<5 20 20>
      constraintSetup {angularLimitsMax=<0 10 60>, angularLimitsMin=<0 -10 0>, angularTarget=<0 0 0>}
      localJointOffset=<5 0 0>
     }]
   }
   n12 = Anim Dynamics - Bone: dyn_coattail_01_l(ComponentPose=@n11, Alpha=1, ExternalForce="0,0,0") {
    angularDampingOverride=0.999
    angularSpringConstant=200
    bAngularSpring=true
    bGravityOverrideInSimSpace=true
    bOverrideAngularDamping=true
    gravityScale=0.35
    simulationSpace=RootRelative
    alphaBoolBlend {blendOption=Linear}
    boundBone {boneName=dyn_coattail_01_l}
    componentLinearAccScale=<0.5 0.5 0.5>
    componentLinearVelScale=<0.5 0.5 0.5>
    physicsBodyDefinitions [{
      sphereCollisionRadius=0
      boundBone {boneName=dyn_coattail_01_l}
      boxExtents=<5 20 20>
      constraintSetup {angularLimitsMax=<0 20 0>, angularLimitsMin=<0 -20 -60>, angularTarget=<0 0 0>}
      localJointOffset=<5 0 0>
     }]
   }
   n13 = Anim Dynamics - Bone: dyn_coattail_02_l(ComponentPose=@n12, Alpha=1, ExternalForce="0,0,0") {
    angularDampingOverride=0.98
    angularSpringConstant=200
    bAngularSpring=true
    bGravityOverrideInSimSpace=true
    bOverrideAngularDamping=true
    gravityScale=0.35
    simulationSpace=RootRelative
    alphaBoolBlend {blendOption=Linear}
    boundBone {boneName=dyn_coattail_02_l}
    physicsBodyDefinitions [{
      sphereCollisionRadius=1
      boundBone {boneName=dyn_coattail_02_l}
      boxExtents=<5 20 20>
      constraintSetup {angularLimitsMax=<0 20 0>, angularLimitsMin=<0 -20 -60>, angularTarget=<0 0 0>}
      localJointOffset=<5 0 0>
     }]
    planarLimits [
     {drivingBone {boneName=thigh_l}, planeTransform {location=<0 -10 0>, rotation=<rot 0 0 -90>}}
     {drivingBone {boneName=calf_l}, planeTransform {location=<0 -10 0>, rotation=<rot 0 0 -90>}}
    ]
    sphericalLimits [
     {limitRadius=10, drivingBone {boneName=thigh_l}, sphereLocalOffset=<-17 0 0>}
     {limitRadius=10, drivingBone {boneName=thigh_l}, sphereLocalOffset=<-25 0 0>}
     {limitRadius=8, drivingBone {boneName=thigh_l}, sphereLocalOffset=<-33 0 0>}
     {limitRadius=10, drivingBone {boneName=thigh_r}, sphereLocalOffset=<17 0 0>}
     {limitRadius=10, drivingBone {boneName=thigh_r}, sphereLocalOffset=<25 0 0>}
     {limitRadius=8, drivingBone {boneName=thigh_r}, sphereLocalOffset=<33 0 0>}
     {limitRadius=10, drivingBone {boneName=calf_l}, sphereLocalOffset=<-8 0 0>}
     {limitRadius=10, drivingBone {boneName=calf_l}, sphereLocalOffset=<-16 0 0>}
     {limitRadius=10, drivingBone {boneName=calf_l}, sphereLocalOffset=<-24 0 0>}
     {limitRadius=10, drivingBone {boneName=calf_r}, sphereLocalOffset=<8 0 0>}
     {limitRadius=10, drivingBone {boneName=calf_r}, sphereLocalOffset=<16 0 0>}
     {limitRadius=10, drivingBone {boneName=calf_r}, sphereLocalOffset=<24 0 0>}
    ]
   }
   n14 = Anim Dynamics - Bone: dyn_coattail_03_l(ComponentPose=@n13, Alpha=1, ExternalForce="0,0,0") {
    angularDampingOverride=0.98
    angularSpringConstant=200
    bAngularSpring=true
    bGravityOverrideInSimSpace=true
    bOverrideAngularDamping=true
    gravityScale=0.35
    simulationSpace=RootRelative
    alphaBoolBlend {blendOption=Linear}
    boundBone {boneName=dyn_coattail_03_l}
    physicsBodyDefinitions [{
      sphereCollisionRadius=1
      boundBone {boneName=dyn_coattail_03_l}
      boxExtents=<5 20 20>
      constraintSetup {angularLimitsMax=<0 20 0>, angularLimitsMin=<0 -20 -60>, angularTarget=<0 0 0>}
      localJointOffset=<5 0 0>
     }]
    planarLimits [
     {drivingBone {boneName=thigh_l}, planeTransform {location=<0 -10 0>, rotation=<rot 0 0 -90>}}
     {drivingBone {boneName=calf_l}, planeTransform {location=<0 -10 0>, rotation=<rot 0 0 -90>}}
    ]
    sphericalLimits [
     {limitRadius=10, drivingBone {boneName=thigh_l}, sphereLocalOffset=<-17 0 0>}
     {limitRadius=10, drivingBone {boneName=thigh_l}, sphereLocalOffset=<-25 0 0>}
     {limitRadius=8, drivingBone {boneName=thigh_l}, sphereLocalOffset=<-33 0 0>}
     {limitRadius=10, drivingBone {boneName=thigh_r}, sphereLocalOffset=<17 0 0>}
     {limitRadius=10, drivingBone {boneName=thigh_r}, sphereLocalOffset=<25 0 0>}
     {limitRadius=8, drivingBone {boneName=thigh_r}, sphereLocalOffset=<33 0 0>}
     {limitRadius=10, drivingBone {boneName=calf_l}, sphereLocalOffset=<-8 0 0>}
     {limitRadius=10, drivingBone {boneName=calf_l}, sphereLocalOffset=<-16 0 0>}
     {limitRadius=10, drivingBone {boneName=calf_l}, sphereLocalOffset=<-24 0 0>}
     {limitRadius=10, drivingBone {boneName=calf_r}, sphereLocalOffset=<8 0 0>}
     {limitRadius=10, drivingBone {boneName=calf_r}, sphereLocalOffset=<16 0 0>}
     {limitRadius=10, drivingBone {boneName=calf_r}, sphereLocalOffset=<24 0 0>}
    ]
   }
   n15 = Anim Dynamics - Bone: dyn_apron_01_l(ComponentPose=@n14, Alpha=1, ExternalForce="0,0,0") {
    angularDampingOverride=0.999
    angularSpringConstant=200
    bAngularSpring=true
    bGravityOverrideInSimSpace=true
    bOverrideAngularDamping=true
    gravityScale=0.35
    simulationSpace=RootRelative
    alphaBoolBlend {blendOption=Linear}
    boundBone {boneName=dyn_apron_01_l}
    componentLinearAccScale=<0.5 0.5 0.5>
    componentLinearVelScale=<0.5 0.5 0.5>
    physicsBodyDefinitions [{
      sphereCollisionRadius=0
      boundBone {boneName=dyn_apron_01_l}
      boxExtents=<5 20 20>
      constraintSetup {angularLimitsMax=<0 0 60>, angularLimitsMin=<0 0 0>, angularTarget=<0 0 0>}
      localJointOffset=<5 0 0>
     }]
   }
   n16 = Anim Dynamics - Bone: dyn_apron_02_l(ComponentPose=@n15, Alpha=1, ExternalForce="0,0,0") {
    angularDampingOverride=0.98
    angularSpringConstant=200
    bAngularSpring=true
    bGravityOverrideInSimSpace=true
    bOverrideAngularDamping=true
    gravityScale=0.35
    simulationSpace=RootRelative
    alphaBoolBlend {blendOption=Linear}
    boundBone {boneName=dyn_apron_02_l}
    physicsBodyDefinitions [{
      sphereCollisionRadius=0
      boundBone {boneName=dyn_apron_02_l}
      boxExtents=<5 20 20>
      constraintSetup {angularLimitsMax=<0 20 60>, angularLimitsMin=<0 -20 0>, angularTarget=<0 0 0>}
      localJointOffset=<5 0 0>
     }]
   }
   n17 = Anim Dynamics - Bone: dyn_apron_03_l(ComponentPose=@n16, Alpha=1, ExternalForce="0,0,0") {
    angularDampingOverride=0.98
    angularSpringConstant=200
    bAngularSpring=true
    bGravityOverrideInSimSpace=true
    bOverrideAngularDamping=true
    gravityScale=0.35
    simulationSpace=RootRelative
    alphaBoolBlend {blendOption=Linear}
    boundBone {boneName=dyn_apron_03_l}
    physicsBodyDefinitions [{
      sphereCollisionRadius=0
      boundBone {boneName=dyn_apron_03_l}
      boxExtents=<5 20 20>
      constraintSetup {angularLimitsMax=<0 10 60>, angularLimitsMin=<0 -10 0>, angularTarget=<0 0 0>}
      localJointOffset=<5 0 0>
     }]
   }
   Output Pose(Result=Blend Poses by bool(
     BlendPose_0=Use cached pose 'CopyPoseCache'()
     BlendPose_1=Component To Local(ComponentPose=@n17)
     BlendTime_0=0
     BlendTime_1=0
     bActiveValue=Get bClothEnabled()
    ) {blendType=Linear}) {layerGroup=None}
   n18 = "Constraint - dyn_vial (PPPPPP)"!(
    ComponentPose=Local To Component(LocalPose=Copy Pose From Mesh() {bUseAttachedParent=true})
    Alpha=1
    ConstraintWeights_0=0
    ConstraintWeights_1:float=0.1424
    ConstraintWeights_2:float=0.303
    ConstraintWeights_3:float=0.3774
    ConstraintWeights_4:float=0.0828
    ConstraintWeights_5:float=0.0944
   ) -> (Pose) {
    alphaBoolBlend {blendOption=Linear}
    boneToModify {boneName=dyn_vial}
    constraintSetup [
     {targetBone {boneName=clavicle_l}}
     {targetBone {boneName=spine_01}}
     {targetBone {boneName=spine_02}}
     {targetBone {boneName=spine_03}}
     {targetBone {boneName=spine_04}}
     {targetBone {boneName=clavicle_r}}
    ]
    constraintWeights [1,1,1,1,1,1]
   }
   n19 = Anim Dynamics - Bone: dyn_vial_liquid(ComponentPose=@n18, Alpha=1, ExternalForce="0,0,0") {
    angularDampingOverride=0.98
    angularSpringConstant=200
    bAngularSpring=true
    bGravityOverrideInSimSpace=true
    bOverrideAngularDamping=true
    gravityScale=0.35
    alphaBoolBlend {blendOption=Linear}
    boundBone {boneName=dyn_vial_liquid}
    componentLinearAccScale=<1 1 1>
    componentLinearVelScale=<1 1 1>
    physicsBodyDefinitions [{
      sphereCollisionRadius=0
      boundBone {boneName=dyn_vial_liquid}
      boxExtents=<2 10 10>
      constraintSetup {angularLimitsMax=<0 45 45>, angularLimitsMin=<0 -45 -45>, angularTarget=<0 0 0>}
      localJointOffset=<-2 0 0>
     }]
   }
   n20 = "Constraint - dyn_coattail_master_r (P)"(ComponentPose=@n19, Alpha=1, ConstraintWeights_0=1) {
    alphaBoolBlend {blendOption=Linear}
    boneToModify {boneName=dyn_coattail_master_r}
    constraintSetup [{targetBone {boneName=thigh_twist_01_r}}]
    constraintWeights [1]
   }
   n21 = "Constraint - dyn_apron_master_r (P)"(ComponentPose=@n20, Alpha=1, ConstraintWeights_0=1) {
    alphaBoolBlend {blendOption=Linear}
    boneToModify {boneName=dyn_apron_master_r}
    constraintSetup [{targetBone {boneName=thigh_twist_01_r}}]
    constraintWeights [1]
   }
   n22 = "Constraint - dyn_coattail_master_l (P)"(ComponentPose=@n21, Alpha=1, ConstraintWeights_0=1) {
    alphaBoolBlend {blendOption=Linear}
    boneToModify {boneName=dyn_coattail_master_l}
    constraintSetup [{targetBone {boneName=thigh_twist_01_l}}]
    constraintWeights [1]
   }
   n23 = "Constraint - dyn_apron_master_l (P)"(ComponentPose=@n22, Alpha=1, ConstraintWeights_0=1) {
    alphaBoolBlend {blendOption=Linear}
    boneToModify {boneName=dyn_apron_master_l}
    constraintSetup [{targetBone {boneName=thigh_twist_01_l}}]
    constraintWeights [1]
   }
   n24 = Anim Dynamics - Bone: dyn_string_r(ComponentPose=@n23, Alpha=1, ExternalForce="0,0,0") {
    angularDampingOverride=0.999
    angularSpringConstant=200
    bAngularSpring=true
    bGravityOverrideInSimSpace=true
    bOverrideAngularDamping=true
    alphaBoolBlend {blendOption=Linear}
    boundBone {boneName=dyn_string_r}
    componentLinearAccScale=<1 1 1>
    componentLinearVelScale=<1 1 1>
    physicsBodyDefinitions [{
      sphereCollisionRadius=0
      boundBone {boneName=dyn_string_r}
      boxExtents=<5 19 21>
      constraintSetup {angularLimitsMax=<0 60 10>, angularLimitsMin=<0 -60 -60>, angularTarget=<0 0 0>}
      localJointOffset=<5 0 0>
     }]
   }
   n25 = Anim Dynamics - Bone: dyn_string_l(ComponentPose=@n24, Alpha=1, ExternalForce="0,0,0") {
    angularDampingOverride=0.999
    angularSpringConstant=200
    bAngularSpring=true
    bGravityOverrideInSimSpace=true
    bOverrideAngularDamping=true
    alphaBoolBlend {blendOption=Linear}
    boundBone {boneName=dyn_string_l}
    componentLinearAccScale=<1 1 1>
    componentLinearVelScale=<1 1 1>
    physicsBodyDefinitions [{
      sphereCollisionRadius=0
      boundBone {boneName=dyn_string_l}
      boxExtents=<5 20 20>
      constraintSetup {angularLimitsMax=<0 60 10>, angularLimitsMin=<0 -60 -60>, angularTarget=<0 0 0>}
      localJointOffset=<3 0 0>
     }]
   }
   n26 = "Constraint - dyn_aprontie_01_l (PPP)"!(ComponentPose=@n25, Alpha=1, ConstraintWeights_0=0.1629, ConstraintWeights_1:float=0.011, ConstraintWeights_2:float=0.8261) -> (Pose) {
    alphaBoolBlend {blendOption=Linear}
    boneToModify {boneName=dyn_aprontie_01_l}
    constraintSetup [{targetBone {boneName=spine_01}}, {targetBone {boneName=spine_02}}, {targetBone {boneName=pelvis}}]
    constraintWeights [1,1,1]
   }
   n27 = "Constraint - dyn_aprontie_01_r (PPP)"!(ComponentPose=@n26, Alpha=1, ConstraintWeights_0=0.1629, ConstraintWeights_1:float=0.011, ConstraintWeights_2:float=0.8261) -> (Pose) {
    alphaBoolBlend {blendOption=Linear}
    boneToModify {boneName=dyn_aprontie_01_r}
    constraintSetup [{targetBone {boneName=spine_01}}, {targetBone {boneName=spine_02}}, {targetBone {boneName=pelvis}}]
    constraintWeights [1,1,1]
   }
   n28 = Anim Dynamics - Bone: dyn_aprontie_01_l(ComponentPose=@n27, Alpha=0.5, ExternalForce="0,0,0") {
    angularDampingOverride=0.999
    angularSpringConstant=200
    bGravityOverrideInSimSpace=true
    bOverrideAngularDamping=true
    simulationSpace=RootRelative
    alphaBoolBlend {blendOption=Linear}
    boundBone {boneName=dyn_aprontie_01_l}
    physicsBodyDefinitions [{
      sphereCollisionRadius=0
      boundBone {boneName=dyn_aprontie_01_l}
      boxExtents=<5 20 20>
      constraintSetup {angularLimitsMax=<0 60 60>, angularLimitsMin=<0 -60 -5>, angularTarget=<0 0 0>}
      localJointOffset=<5 0 0>
     }]
   }
   n29 = Anim Dynamics - Bone: dyn_aprontie_02_l(ComponentPose=@n28, Alpha=0.5, ExternalForce="0,0,0") {
    angularDampingOverride=0.999
    angularSpringConstant=200
    bGravityOverrideInSimSpace=true
    bOverrideAngularDamping=true
    simulationSpace=RootRelative
    alphaBoolBlend {blendOption=Linear}
    boundBone {boneName=dyn_aprontie_02_l}
    physicsBodyDefinitions [{
      sphereCollisionRadius=0
      boundBone {boneName=dyn_aprontie_02_l}
      boxExtents=<5 20 20>
      constraintSetup {angularLimitsMax=<0 60 60>, angularLimitsMin=<0 -60 -5>, angularTarget=<0 0 0>}
      localJointOffset=<5 0 0>
     }]
   }
   n30 = Anim Dynamics - Bone: dyn_aprontie_01_r(ComponentPose=@n29, Alpha=0.5, ExternalForce="0,0,0") {
    angularDampingOverride=0.999
    angularSpringConstant=200
    bGravityOverrideInSimSpace=true
    bOverrideAngularDamping=true
    simulationSpace=RootRelative
    alphaBoolBlend {blendOption=Linear}
    boundBone {boneName=dyn_aprontie_01_r}
    physicsBodyDefinitions [{
      sphereCollisionRadius=0
      boundBone {boneName=dyn_aprontie_01_r}
      boxExtents=<5 20 20>
      constraintSetup {angularLimitsMax=<0 60 60>, angularLimitsMin=<0 -60 -5>, angularTarget=<0 0 0>}
      localJointOffset=<5 0 0>
     }]
   }
   n31 = Anim Dynamics - Bone: dyn_aprontie_02_r(ComponentPose=@n30, Alpha=0.5, ExternalForce="0,0,0") {
    angularDampingOverride=0.999
    angularSpringConstant=200
    bGravityOverrideInSimSpace=true
    bOverrideAngularDamping=true
    simulationSpace=RootRelative
    alphaBoolBlend {blendOption=Linear}
    boundBone {boneName=dyn_aprontie_02_r}
    physicsBodyDefinitions [{
      sphereCollisionRadius=0
      boundBone {boneName=dyn_aprontie_02_r}
      boxExtents=<5 20 20>
      constraintSetup {angularLimitsMax=<0 60 60>, angularLimitsMin=<0 -60 -5>, angularTarget=<0 0 0>}
      localJointOffset=<5 0 0>
     }]
   }
   n32 = Anim Dynamics - Bone: dyn_aprontie_03_r(ComponentPose=@n31, Alpha=0.5, ExternalForce="0,0,0") {
    angularDampingOverride=0.999
    angularSpringConstant=200
    bGravityOverrideInSimSpace=true
    bOverrideAngularDamping=true
    simulationSpace=RootRelative
    alphaBoolBlend {blendOption=Linear}
    boundBone {boneName=dyn_aprontie_03_r}
    physicsBodyDefinitions [{
      sphereCollisionRadius=0
      boundBone {boneName=dyn_aprontie_03_r}
      boxExtents=<5 20 20>
      constraintSetup {angularLimitsMax=<0 60 60>, angularLimitsMin=<0 -60 -5>, angularTarget=<0 0 0>}
      localJointOffset=<5 0 0>
     }]
   }
   Save cached pose 'CopyPoseCache'(Pose=Component To Local(ComponentPose=@n32)) {cacheName=CopyPoseCache}
  ]}
 EventGraph {nodes [
   TryGetPawnOwner()
   Event BlueprintUpdateAnimation()
  ]}
]
variables []
